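; Sniper weapon Blueprint for a mark-and-execute rifle, giving local visual and audio feedback when aimed at a marked enemy player.
; While targeting, a fast looping timer line-traces from the view point for MarkExecute.MarkedEnemy-tagged players to update isMarked.
; That state drives a timeline fading an attached Niagara effect and scope post-process material, and selects mark or execute fire sounds.
; A mod-added handler checks for iron sights, which enable fire camera shake while ADS.
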(event EventBeginPlay
  ;[1 1 1] BeginPlay: If we have Iron Sights, set Bool to True, which will enable camera shake when fired while ADS in 3P
  (|CallParentFunction)
  (WeaponMods|AssignOnModAddedtoWeapon (delegate Custom|OnModAddedToWeapon_Event)))

(event Custom|OnModAddedToWeapon_Event
  (params (Mod :type Fort%20Weapon%20Mod%20Item%20Definition%20Object%20Reference) (Weapon :type Fort%20Weapon%20Object%20Reference))
  ;[1 1 1] BeginPlay: If we have Iron Sights, set Bool to True, which will enable camera shake when fired while ADS in 3P
  (bind _array_element (Utilities|Array|ForEachLoopwithBreak (CallFunction|GetWeaponModSlots))
    (:LoopBody
      (if (== (Variables|/Script/FortniteGame.FortWeaponModItemDefinition|GetPluginTuningTag (Utilities|Struct|BreakFortWeaponModSlot _array_element)) (Variables|Default|GetIronSightsModTag_0))
        (Variables|Default|SetHasIronSights_0 true))
      (Variables|Default|SetHasIronSights_0 false))
    (:Completed
      _)))

(event EventOnPlayWeaponFireFX
  (params (bPersistentFire :type Boolean) (bSecondaryFire :type Boolean))
  ;[1 1 1] PlayWeaponFireFX: Play Camera Shake if we're ADS with Iron Sights
  (|CallParentFunction bPersistentFire bSecondaryFire)
  (Utilities|FlowControl|Sequence
    (:then_0
      (if (and (Variables|Default|GetHasIronSights_0) (FortWeapon|IsTargeting))
        (CameraShakes|StartCameraShake :self (Variables|/Script/Engine.PlayerController|GetPlayerCameraManager (Pawn|GetLocalViewingPlayerController (Utilities|Casting|CastToFortPlayerPawn (Game|GetInstigator)))) :ShakeClass (Variables|Default|GetFireCameraShake_0) :Scale 1.0 :PlaySpace "CameraLocal")))))

(event EventPlayWeaponFireCustomAudio
  (params (bPersistentFire :type Boolean) (bSecondaryFire :type Boolean))
  ; PlayWeaponFireCustomAudio: Plays mark fire and execute fire sweetener sound
  (|CallParentFunction bPersistentFire bSecondaryFire)
  (if (FortWeapon|IsTargeting)
    (if (Variables|Default|GetisMarked)
      (Default|PlayEventonLocalPlayer "(TagName=\"SoundLibrary.Weapon.Effect.ExecuteFire\")")
      (else
        (Default|PlayEventonLocalPlayer "(TagName=\"SoundLibrary.Weapon.Effect.MarkFire\")")))))

(event EventOnSetTargeting
  (params (bNewIsTargeting :type Boolean))
  (bind _markedfxsystem (Variables|Default|GetMarkedFXSystem))
  (|CallParentFunction bNewIsTargeting)
  ;[0.15 0.15 0.15 0.5 14] Post Process - Enable Mark & Execute functionality
  (CallFunction|PostProcessEnableMark)
  (bind _output_get (Variables|Default|SetPlayer%20is%20ADS_0 bNewIsTargeting))
  (if _output_get
    (CallFunction|CheckifValidforExecute)
    (else
      (Utilities|Time|ClearandInvalidateTimerbyHandle (Variables|Default|GetLooping%20Timer%20Check%20for%20Tag_0))
      ;[0.8281 0.115 0.09919 0.5] FX - Show Execute State
      (Variables|Default|SetisMarked false)
      (bind (_direction _fadeinout _post_process) (Timeline|MarkedFXTimeline ReverseFromEnd)
        (:Update
          (Niagara|SetNiagaraVariable(Float) _markedfxsystem "FXOpacity" _fadeinout)
          ;[0.15 0.07289 0.08028 0.5] Post Process - Show Execute State
          (bind _returnvalue (Scope|GetorCreateScopePostProcessMaterialInstanceDynamics))
          (bind _asmaterial_instance_dynamic (Utilities|Casting|CastToMaterialInstanceDynamic (.Object (Utilities|Struct|BreakWeightedBlendable (Utilities|Array|Get(aref) _returnvalue)))))
          (Rendering|Material|SetScalarParameterValue _asmaterial_instance_dynamic "Marked" _post_process :!class /Script/Engine.MaterialInstanceDynamic))
        (:Finished
          (if (not (Variables|Default|GetisMarked))
            (Components|Activation|Deactivate _markedfxsystem)))))))

(event Custom|CheckIfValidForExecute
  (Utilities|Time|ClearandInvalidateTimerbyHandle (Variables|Default|GetLooping%20Timer%20Check%20for%20Tag_0))
  (bind _returnvalue (Utilities|Time|SetTimerbyEvent (delegate Custom|CheckIfValid) 0.01 true))
  (Variables|Default|SetLooping%20Timer%20Check%20for%20Tag_0 _returnvalue))

(event Custom|CheckIfValid
  (bind (_location _rotation) (Pawn|GetPlayerViewPoint (Fortnite|GetFortPlayerControllerfromActor (Actor|GetOwner))))
  (bind _markedfxsystem (Variables|Default|GetMarkedFXSystem))
  (bind _outhit (Collision|LineTraceByChannel _location (+ _location (* (Math|Vector|GetForwardVector _rotation) 2e+04)) "TraceTypeQuery1" false (Utilities|Array|MakeArray (Actor|GetOwner))))
  (bind _bdidactorhavetags (Class|AthenaQuestFunctionLibrary|AthenaCheckActorHasTags (.HitActor (Collision|BreakHitResult _outhit)) "(GameplayTags=((TagName=\"Pawn.Athena.Player\"),(TagName=\"MarkExecute.MarkedEnemy\")))" true))
  (if (and _bdidactorhavetags (Collision|BreakHitResult _outhit))
    ;[0.8281 0.115 0.09919 0.5] FX - Show Execute State
    (if (not (Variables|Default|GetisMarked))
      (Variables|Default|SetisMarked true)
      (Default|PlayEventonLocalPlayer "(TagName=\"SoundLibrary.Weapon.Effect.ExecuteOn\")")
      (if (Components|Activation|IsActive (Variables|Default|GetMarkedFXSystem))
        (Timeline|MarkedFXTimeline PlayFromStart)
        (else
          (bind _returnvalue (Niagara|SpawnSystemAttached "/VividRazorWeaponGameplay/Gameplay/MarkExecuteSniper/FX/NS_Sniper_MarkExecuteState.NS_Sniper_MarkExecuteState" (Variables|Visual|GetWeaponMesh) "None" "0.000000,90.000000,10.500000"))
          (Variables|Default|SetMarkedFXSystem _returnvalue)
          (Timeline|MarkedFXTimeline PlayFromStart))))
    (else
      (if (Variables|Default|GetisMarked)
        (Default|PlayEventonLocalPlayer "(TagName=\"SoundLibrary.Weapon.Effect.MarkOff\")")
        (Variables|Default|SetisMarked false)
        (Timeline|MarkedFXTimeline ReverseFromEnd)))))

(fn PlayEventOnLocalPlayer
  (params (SoundLibraryEvent :type Sound%20Library%20Tag%20Structure))
  (bind _returnvalue (Game|GetInstigator))
  (Utilities|IsValid _returnvalue
    (:Is%20Valid
      (if (Pawn|IsLocallyViewed _returnvalue)
        (SoundLibrary|PlayEvent (Variables|SoundLibrary|GetSoundLibraryComponent) (Variables|GetSoundLibraryEvent))))))

(fn PostProcessEnableMark
  (bind _returnvalue (Scope|GetorCreateScopePostProcessMaterialInstanceDynamics))
  (bind _asmaterial_instance_dynamic (Utilities|Casting|CastToMaterialInstanceDynamic (.Object (Utilities|Struct|BreakWeightedBlendable (Utilities|Array|Get(aref) _returnvalue)))))
  (Rendering|Material|SetScalarParameterValue _asmaterial_instance_dynamic "Use Mark" 1.0 :!class /Script/Engine.MaterialInstanceDynamic))

(fn ConstructionScript
  (|CallParentFunction))
