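; Event-triggered gameplay ability for the mark-execute sniper that marks hostile pawns hit and applies the EagleEye damage buff to its owner.
; It filters the event target by affiliation, then adds indicator and stencil highlights for the owning controller.
; AI targets also receive a marking effect, with quest events and a tag-gated buff application.
; An impact cue switches between marked and eliminate variants using the tracker weapon's isMarked state.
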(event Ability|EventActivateAbilityFromEvent
  (params (EventData :type Gameplay%20Event%20Data%20Structure%20%28by%20ref%29))
  (bind (_eventtag _instigator _target _optionalobject _optionalobject2 _contexthandle _instigatortags _targettags _eventmagnitude _targetdata) (Utilities|Struct|BreakGameplayEventData EventData))
  (bind _returnvalue (Ability|GetAvatarActorfromActorInfo))
  (bind _returnvalue_2 (Fortnite|GetFortPlayerControllerfromActor _returnvalue))
  (bind _returnvalue_3 (Quests|GetQuestManager _returnvalue_2 "Athena"))
  (bind _output (Utilities|Array|Get(acopy) (Variables|TrackingPistol|Dynamic|GetHitActors)))
  (bind _hitactors (Variables|TrackingPistol|Dynamic|GetHitActors))
  (Ability|ApplyGameplayEffectToOwner "/VividRazorWeaponGameplay/Gameplay/MarkExecuteSniper/GE_DamageBuff_EagleEye.GE_DamageBuff_EagleEye_C")
  (bind _outhitactors (Utilities|HasValidGroupOfHitActors (CallFunction|GetHitActorsfromEventData EventData))
    (:true
      (Variables|TrackingPistol|Dynamic|SetHitActors _outhitactors)
      (bind _returnvalue_4 (Utilities|GetValidAvatarasFortPlayerController true)
        (:Is%20Valid
          (bind _structout (Utilities|Struct|SetmembersinIndicatedActorData (Variables|Default|GetIndicated%20Actor%20Data) (IndicatedManagement|MakeIndicatedActorScaleandOpacityData 7.5e+03 1e+01 0.7 0.9) _instigatortags))
          (IndicatedManagement|AddActorsToIndicatedList _returnvalue_4 _hitactors _structout false false false true)
          (IndicatedManagement|AddActorsToStenciledList _returnvalue_4 _hitactors (Variables|Default|GetStenciled%20Actor%20Data))
          (Utilities|Casting|CastToFortPlayerPawnAthena _output
            (:then
              _)
            (:CastFailed
              (Utilities|Casting|CastToFortAIPawn _output
                (:then
                  (Fortnite|ApplyGameplayEffectsToActor _output (Variables|TrackingPistol|Default|Indicator|GetEnemyMarkingGE) (Transformation|GetActorLocation _returnvalue_4) false 0 (Ability|GetAbilitySystemComponent _returnvalue_4))
                  (Utilities|IsValid _returnvalue
                    (:Is%20Valid
                      (Utilities|IsValid _returnvalue_2
                        (:Is%20Valid
                          (Utilities|IsValid _returnvalue_3
                            (:Is%20Valid
                              (Utilities|FlowControl|Sequence
                                (:then_0
                                  ;[1 1 1] Adding gameplay event for QS2
                                  (Ability|SendGameplayEventtoActor _returnvalue_2 "(TagName=\"Event.MarkedOpponent\")" (Utilities|Struct|MakeGameplayEventData "(TagName=\"\")" _instigator _target 0 0 _contexthandle _instigatortags _targettags 1.0 _targetdata))
                                  (bind _bdidactorhavetags (Class|AthenaQuestFunctionLibrary|AthenaCheckActorHasTags _output "(GameplayTags=((TagName=\"Athena.Quests.MarkedEnemy\")))" true))
                                  (if _bdidactorhavetags
                                    (Ability|ApplyGameplayEffectToTarget :GameplayEffectClass "/VividRazorWeaponGameplay/Gameplay/MarkExecuteSniper/GE_DamageBuff_EagleEye.GE_DamageBuff_EagleEye_C" :GameplayEffectLevel 1 :Stacks 1)))
                                (:then_1
                                  (Quest|SendComplexCustomStatEvent _returnvalue_3 0 "(GameplayTags=((TagName=\"Athena.Quests.MarkedEnemy\")))" "(GameplayTags=)" 1))
                                (:then_2
                                  (CallFunction|PlayTargetImpactSweetner _output))))))))))
                (:CastFailed
                  _))
              (Ability|EndAbility))))))
    (:false
      (Ability|EndAbility))))

(fn GetAvatarAsFortPawn
  (bind _asfort_pawn (Utilities|Casting|CastToFortPawn (Ability|GetAvatarActorfromActorInfo))
    (:then
      (return _asfort_pawn))
    (:CastFailed
      (return 0))))

(fn GetAvatarAsFortPlayerController
  (bind _outfortplayerpawn (CallFunction|GetAvatarasValidFortPlayerPawn)
    (:ValidFortPlayerPawn
      (Utilities|IsValid (Fortnite|GetFortPlayerControllerfromActor _outfortplayerpawn)
        (:Is%20Valid
          (return (Fortnite|GetFortPlayerControllerfromActor _outfortplayerpawn)))
        (:Is%20Not%20Valid
          _)))
    (:AvatarCastFailed
      _))
  (return 0))

(fn GetHitActorsFromEventData
  (params (GameplayEventData :type Gameplay%20Event%20Data%20Structure%20%28by%20ref%29))
  (bind (_eventtag _instigator _target) (Utilities|Struct|BreakGameplayEventData GameplayEventData))
  (bind _output_get (LocalVariables|SetLocalTarget _target))
  (Utilities|IsValid _output_get
    (:Is%20Valid
      (bind _returnvalue (Utilities|GetValidAvatarasFortPlayerController true)
        (:Is%20Valid
          (if (CallFunction|IsPawnandNotFriendly _returnvalue (LocalVariables|GetLocalTarget))
            (return (Utilities|Array|MakeArray (LocalVariables|GetLocalTarget)))
            (else
              (return 0))))))
    (:Is%20Not%20Valid
      (return 0))))

(fn IsDamgeFromSourceWeapon
  (params (GameplayEventData :type Gameplay%20Event%20Data%20Structure%20%28by%20ref%29))
  (bind (_eventtag _instigator _target _optionalobject _optionalobject2 _contexthandle) (Utilities|Struct|BreakGameplayEventData GameplayEventData))
  (bind _asfort_pawn (CallFunction|GetAvatarasFortPawn))
  (Utilities|IsValid _asfort_pawn
    (:Is%20Valid
      (return (Utilities|Equal(Object) (Variables|/Script/FortniteGame.FortPawn|GetCurrentWeapon _asfort_pawn) (Ability|EffectContext|GetSourceObject _contexthandle))))
    (:Is%20Not%20Valid
      (return false))))

(fn IsPawnAndNotFriendly
  (params (OwningController :type Actor%20Object%20Reference) (HitActor :type Actor%20Object%20Reference))
  (bind (_asfort_aipawn _bsuccess) (Utilities|Casting|CastToFortAIPawn HitActor))
  (bind (_asfort_player_pawn _bsuccess_2) (Utilities|Casting|CastToFortPlayerPawn HitActor))
  (return (and (Utilities|Enum|Equal(Enum) (Team|GetTeamAffiliation OwningController HitActor) "Hostile") (or _bsuccess_2 _bsuccess))))

(fn PlayTargetImpactSweetner
  (params (TargetActor :type Actor%20Object%20Reference))
  (bind _output_get (LocalVariables|SetInstigating%20Pawn (CallFunction|GetAvatarasFortPawn)))
  (bind _asb_sniper_tracker (Utilities|Casting|CastToB_Sniper_Tracker (Variables|/Script/FortniteGame.FortPawn|GetCurrentWeapon _output_get)))
  (GameplayCue|ExecuteGameplayCueOnActor(Burst) (Variables|GetTargetActor) (select (Variables|/VividRazorWeaponGameplay/Gameplay/MarkExecuteSniper/B_Sniper_Tracker.B_Sniper_Tracker_C|GetisMarked _asb_sniper_tracker) (Variables|Default|GetEliminate%20Cue) (Variables|Default|GetMarked%20Cue)) (Ability|GameplayCue|MakeGameplayCueParameters :NormalizedMagnitude 0.0 :RawMagnitude 0.0 :EffectContext (Class|AthenaFunctionLibrary|MakeSimpleFortEffectContext (CallFunction|GetAvatarasFortPawn) (CallFunction|GetAvatarasFortPawn)) :Location "0,0,0" :Normal "0,0,0" :Instigator (CallFunction|GetAvatarasFortPawn) :EffectCauser (CallFunction|GetAvatarasFortPawn) :SourceObject (Fortnite|GetControllerfromActor (Variables|GetTargetActor)) :GameplayEffectLevel 1 :AbilityLevel 1 :bReplicateLocationWhenUsingMinimalRepProxy false)))
