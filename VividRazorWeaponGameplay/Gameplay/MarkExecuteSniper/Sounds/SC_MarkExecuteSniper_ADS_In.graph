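# Sound cue for the MarkExecute sniper's aim-down-sights-in foley.
# A mixer layers two wave players at equal volume: a shared MoonFlax sniper ADS-in wave and a MarkExecute-specific ADSing wave.
# It plays quietly in the weapon foley sound class, using the weapon reload attenuation settings, and has the debug flag enabled.
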SoundCue /VividRazorWeaponGameplay/Gameplay/MarkExecuteSniper/Sounds/SC_MarkExecuteSniper_ADS_In
root=Mixer_0
nodes [
 {id=Mixer_0, children [WavePlayer_0,WavePlayer_1], properties {inputVolume [1,1]}}
 {id=WavePlayer_0, properties {soundWaveAssetPtr=/MoonFlaxWeaponGameplay/Items/Weapons/Sniper/SoundsVS/Waves/Sniper_MoonFlax_ADS_In_01}}
 {id=WavePlayer_1, properties {soundWaveAssetPtr=/VividRazorWeaponGameplay/Gameplay/MarkExecuteSniper/Sounds/wavs/MarkExecuteSniper_ADSing}}
]
properties {
 attenuationSettings=&Fort_Weapon_Reload_Att
 bDebug=true
 maxDistance=2250
 soundClassObject=&Fort_WeaponFoley
 volumeMultiplier=0.2
}
