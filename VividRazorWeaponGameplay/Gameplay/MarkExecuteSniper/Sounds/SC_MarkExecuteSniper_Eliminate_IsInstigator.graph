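# Sound cue for the Mark Execute Sniper's elimination sound, named as the variant for the player who instigated the elimination.
# A single wave player feeds the root directly, with no randomization or mixing.
# It plays at reduced volume through the critical in-game UI sound class, sets no attenuation asset, and has its debug flag enabled.
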SoundCue /VividRazorWeaponGameplay/Gameplay/MarkExecuteSniper/Sounds/SC_MarkExecuteSniper_Eliminate_IsInstigator
root=WavePlayer_2
nodes [{
  id=WavePlayer_2
  properties {soundWaveAssetPtr=/VividRazorWeaponGameplay/Gameplay/MarkExecuteSniper/Sounds/wavs/MarkExecuteSniper_Eliminate_IsInstigator}
 }]
properties {bDebug=true, maxDistance=2097152, soundClassObject=&Fort_UI_InGame_Critical, volumeMultiplier=0.5}
