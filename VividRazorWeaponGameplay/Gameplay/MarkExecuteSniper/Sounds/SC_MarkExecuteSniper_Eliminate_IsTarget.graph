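# Sound cue for the Mark Execute Sniper weapon's elimination sound when the victim is the target, as its name indicates.
# A single wave player feeds the root with one wav, with no randomization or modulation.
# It uses the critical in-game UI sound class and overrides concurrency to one owner-limited instance with a short retrigger time.
# The cue has its debug flag left on.
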SoundCue /VividRazorWeaponGameplay/Gameplay/MarkExecuteSniper/Sounds/SC_MarkExecuteSniper_Eliminate_IsTarget
root=WavePlayer_0
nodes [{
  id=WavePlayer_0
  properties {soundWaveAssetPtr=/VividRazorWeaponGameplay/Gameplay/MarkExecuteSniper/Sounds/wavs/MarkExecuteSniper_Eliminate_IsTarget}
 }]
properties {
 bDebug=true
 bOverrideConcurrency=true
 maxDistance=2097152
 soundClassObject=&Fort_UI_InGame_Critical
 concurrencyOverrides {bLimitToOwner=true, maxCount=1, retriggerTime=0.1}
}
