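# Sound cue for the Mark Execute Sniper weapon, named for its execute mode turning off.
# A single wave player feeds the MarkExecuteSniper_ExecuteModeOff wave straight to the output, with no randomization or mixing.
# It plays at reduced volume under the weapon foley sound class with Fortnite's weapon reload attenuation.
# The cue's debug flag is switched on.
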SoundCue /VividRazorWeaponGameplay/Gameplay/MarkExecuteSniper/Sounds/SC_MarkExecuteSniper_ExecuteModeOff
root=WavePlayer_0
nodes [{
  id=WavePlayer_0
  properties {soundWaveAssetPtr=/VividRazorWeaponGameplay/Gameplay/MarkExecuteSniper/Sounds/wavs/MarkExecuteSniper_ExecuteModeOff}
 }]
properties {
 attenuationSettings=&Fort_Weapon_Reload_Att
 bDebug=true
 maxDistance=2250
 soundClassObject=&Fort_WeaponFoley
 volumeMultiplier=0.45
}
