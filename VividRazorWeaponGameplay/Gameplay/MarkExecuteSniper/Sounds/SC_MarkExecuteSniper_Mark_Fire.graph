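# Sound cue for the firing sound of the Mark Execute Sniper's mark shot.
# A single wave player feeds the MarkExecuteSniper_Mark_Fire wave straight to the output, with no randomization or modulation.
# It uses the ranged weapon sound class and close-projectile attenuation at slightly reduced volume, and has its debug flag switched on.
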SoundCue /VividRazorWeaponGameplay/Gameplay/MarkExecuteSniper/Sounds/SC_MarkExecuteSniper_Mark_Fire
root=WavePlayer_0
nodes [{id=WavePlayer_0, properties {soundWaveAssetPtr=/VividRazorWeaponGameplay/Gameplay/MarkExecuteSniper/Sounds/wavs/MarkExecuteSniper_Mark_Fire}}]
properties {
 attenuationSettings=&Fort_Weapon_Projectile_Close_Athena_Att
 bDebug=true
 maxDistance=4950
 soundClassObject=&Fort_WeaponRanged
 volumeMultiplier=0.7
}
