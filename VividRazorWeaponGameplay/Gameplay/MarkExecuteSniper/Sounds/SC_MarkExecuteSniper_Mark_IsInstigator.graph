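# Sound cue for the Mark Execute Sniper's mark sound, named as the variant heard by the instigator of the mark.
# A single wave player feeds the output directly, with no randomization, mixing or modulation nodes.
# It routes through the critical in-game UI sound class, has a very large max distance, and has its debug flag left on.
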SoundCue /VividRazorWeaponGameplay/Gameplay/MarkExecuteSniper/Sounds/SC_MarkExecuteSniper_Mark_IsInstigator
root=WavePlayer_1
nodes [{
  id=WavePlayer_1
  properties {soundWaveAssetPtr=/VividRazorWeaponGameplay/Gameplay/MarkExecuteSniper/Sounds/wavs/MarkExecuteSniper_Mark_IsInstigator}
 }]
properties {bDebug=true, maxDistance=2097152, soundClassObject=&Fort_UI_InGame_Critical}
