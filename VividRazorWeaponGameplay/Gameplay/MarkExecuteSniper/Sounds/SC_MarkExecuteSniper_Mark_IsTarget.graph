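# Sound cue for the Mark Execute Sniper weapon, named as the alert for being the target of a mark.
# A single wave player feeds the IsTarget wave straight to the output, with no randomization, mixing or modulation.
# It routes through the critical in-game UI sound class, has a very large max distance, and has its debug flag enabled.
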SoundCue /VividRazorWeaponGameplay/Gameplay/MarkExecuteSniper/Sounds/SC_MarkExecuteSniper_Mark_IsTarget
root=WavePlayer_3
nodes [{
  id=WavePlayer_3
  properties {soundWaveAssetPtr=/VividRazorWeaponGameplay/Gameplay/MarkExecuteSniper/Sounds/wavs/MarkExecuteSniper_Mark_IsTarget}
 }]
properties {bDebug=true, maxDistance=2097152, soundClassObject=&Fort_UI_InGame_Critical}
